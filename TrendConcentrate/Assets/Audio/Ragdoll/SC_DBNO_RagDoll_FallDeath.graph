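# Sound cue for a DBNO ragdoll fall-to-death, built as a layered, randomized one-shot.
# A mixer plays three layers at once (A, B and C wave sets), each picking an equally weighted random variation, with the C layer slightly quieter.
# It uses a dedicated attenuation asset, the Fort_Gadgets sound class, and two concurrency sets, one fall-death 1P and one DBNO roll 3P.
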SoundCue /TrendConcentrate/Assets/Audio/Ragdoll/SC_DBNO_RagDoll_FallDeath
root=Mixer_0
nodes [
 {id=Mixer_0, children [Random_1,Random_2,Random_3], properties {inputVolume [1,1,0.7]}}
 {id=Random_1, children [WavePlayer_3,WavePlayer_4,WavePlayer_5], properties {weights [1,1,1]}}
 {id=Random_2, children [WavePlayer_6, WavePlayer_7, WavePlayer_8, WavePlayer_9, WavePlayer_10], properties {weights [1,1,1,1,1]}}
 {
  id=Random_3
  children [WavePlayer_11, WavePlayer_12, WavePlayer_13, WavePlayer_14, WavePlayer_15, WavePlayer_16]
  properties {weights [1,1,1,1,1,1]}
 }
 {id=WavePlayer_10, properties {soundWaveAssetPtr=/TrendConcentrate/Assets/Audio/Ragdoll/Waves/FalltoDeath/DBNO_RagDoll_FallDeathB_05}}
 {id=WavePlayer_11, properties {soundWaveAssetPtr=/TrendConcentrate/Assets/Audio/Ragdoll/Waves/FalltoDeath/DBNO_RagDoll_FallDeathC_01}}
 {id=WavePlayer_12, properties {soundWaveAssetPtr=/TrendConcentrate/Assets/Audio/Ragdoll/Waves/FalltoDeath/DBNO_RagDoll_FallDeathC_02}}
 {id=WavePlayer_13, properties {soundWaveAssetPtr=/TrendConcentrate/Assets/Audio/Ragdoll/Waves/FalltoDeath/DBNO_RagDoll_FallDeathC_03}}
 {id=WavePlayer_14, properties {soundWaveAssetPtr=/TrendConcentrate/Assets/Audio/Ragdoll/Waves/FalltoDeath/DBNO_RagDoll_FallDeathC_04}}
 {id=WavePlayer_15, properties {soundWaveAssetPtr=/TrendConcentrate/Assets/Audio/Ragdoll/Waves/FalltoDeath/DBNO_RagDoll_FallDeathC_05}}
 {id=WavePlayer_16, properties {soundWaveAssetPtr=/TrendConcentrate/Assets/Audio/Ragdoll/Waves/FalltoDeath/DBNO_RagDoll_FallDeathC_06}}
 {id=WavePlayer_3, properties {soundWaveAssetPtr=/TrendConcentrate/Assets/Audio/Ragdoll/Waves/FalltoDeath/DBNO_RagDoll_FallDeathA_01}}
 {id=WavePlayer_4, properties {soundWaveAssetPtr=/TrendConcentrate/Assets/Audio/Ragdoll/Waves/FalltoDeath/DBNO_RagDoll_FallDeathA_02}}
 {id=WavePlayer_5, properties {soundWaveAssetPtr=/TrendConcentrate/Assets/Audio/Ragdoll/Waves/FalltoDeath/DBNO_RagDoll_FallDeathA_03}}
 {id=WavePlayer_6, properties {soundWaveAssetPtr=/TrendConcentrate/Assets/Audio/Ragdoll/Waves/FalltoDeath/DBNO_RagDoll_FallDeathB_01}}
 {id=WavePlayer_7, properties {soundWaveAssetPtr=/TrendConcentrate/Assets/Audio/Ragdoll/Waves/FalltoDeath/DBNO_RagDoll_FallDeathB_02}}
 {id=WavePlayer_8, properties {soundWaveAssetPtr=/TrendConcentrate/Assets/Audio/Ragdoll/Waves/FalltoDeath/DBNO_RagDoll_FallDeathB_03}}
 {id=WavePlayer_9, properties {soundWaveAssetPtr=/TrendConcentrate/Assets/Audio/Ragdoll/Waves/FalltoDeath/DBNO_RagDoll_FallDeathB_04}}
]
properties {
 attenuationSettings=&/ATT_DBNO_RagDoll_FallDeath
 maxDistance=12000
 soundClassObject=&Fort_Gadgets
 volumeMultiplier=1
 concurrencySet [&/CONC_DBNO_FallDeath_1P, &/CS_DBNO_Roll_3p]
}
